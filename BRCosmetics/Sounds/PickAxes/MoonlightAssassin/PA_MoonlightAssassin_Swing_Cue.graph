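# Sound cue for the Moonlight Assassin pickaxe swing.
# A quality-level switch selects one of three mixes, stepping from six random variations per layer to three, then to one shared by the last two levels.
# Each mix layers custom swing waves under close attenuation with pitch-randomized generic fire axe distant swings under medium attenuation.
# Concurrency overrides limit playback per owner and stop the oldest sound.
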SoundCue /BRCosmetics/Sounds/PickAxes/MoonlightAssassin/PA_MoonlightAssassin_Swing_Cue
root=QualityLevel_2
nodes [
 {id=Attenuation_12, children [Modulator_14], properties {attenuationSettings=/Game/Sounds/Attenuation/Fort_Weapon_Pickaxe_Close_Att}}
 {id=Attenuation_13, children [Modulator_12], properties {attenuationSettings=/Game/Sounds/Attenuation/Fort_Weapon_Pickaxe_Medium_Att}}
 {id=Attenuation_14, children [Modulator_16], properties {attenuationSettings=&Fort_Weapon_Pickaxe_Close_Att}}
 {id=Attenuation_15, children [Modulator_13], properties {attenuationSettings=&Fort_Weapon_Pickaxe_Medium_Att}}
 {id=Attenuation_16, children [WavePlayer_59], properties {attenuationSettings=&Fort_Weapon_Pickaxe_Close_Att}}
 {id=Attenuation_17, children [Modulator_11], properties {attenuationSettings=&Fort_Weapon_Pickaxe_Medium_Att}}
 {id=Mixer_6, children [Attenuation_12, Attenuation_13], properties {inputVolume [1,1]}}
 {id=Mixer_7, children [Attenuation_14, Attenuation_15], properties {inputVolume [1,1]}}
 {id=Mixer_8, children [Attenuation_16, Attenuation_17], properties {inputVolume [1,1]}}
 {id=Modulator_11, children [WavePlayer_49], properties {pitchMax=1.2, pitchMin=0.9}}
 {id=Modulator_12, children [Random_8], properties {pitchMax=1.2, pitchMin=0.9}}
 {id=Modulator_13, children [Random_9], properties {pitchMax=1.2, pitchMin=0.9}}
 {id=Modulator_14, children [Random_10], properties {pitchMax=1}}
 {id=Modulator_16, children [Random_11], properties {pitchMax=1}}
 {id=QualityLevel_2, children [Mixer_6,Mixer_7,Mixer_8,Mixer_8]}
 {
  id=Random_10
  children [WavePlayer_50, WavePlayer_51, WavePlayer_52, WavePlayer_53, WavePlayer_54, WavePlayer_55]
  properties {weights [1,1,1,1,1,1]}
 }
 {id=Random_11, children [WavePlayer_56, WavePlayer_57, WavePlayer_58], properties {weights [1,1,1]}}
 {
  id=Random_8
  children [WavePlayer_40, WavePlayer_41, WavePlayer_42, WavePlayer_43, WavePlayer_44, WavePlayer_45]
  properties {weights [1,1,1,1,1,1]}
 }
 {id=Random_9, children [WavePlayer_47, WavePlayer_46, WavePlayer_48], properties {weights [1,1,1]}}
 {id=WavePlayer_40, properties {soundWaveAssetPtr=/Game/Athena/Sounds/Weapons/Fire_Axe_Swing_Distant_04}}
 {id=WavePlayer_41, properties {soundWaveAssetPtr=/Game/Athena/Sounds/Weapons/Fire_Axe_Swing_Distant_01}}
 {id=WavePlayer_42, properties {soundWaveAssetPtr=/Game/Athena/Sounds/Weapons/Fire_Axe_Swing_Distant_02}}
 {id=WavePlayer_43, properties {soundWaveAssetPtr=/Game/Athena/Sounds/Weapons/Fire_Axe_Swing_Distant_03}}
 {id=WavePlayer_44, properties {soundWaveAssetPtr=/Game/Athena/Sounds/Weapons/Fire_Axe_Swing_Distant_05}}
 {id=WavePlayer_45, properties {soundWaveAssetPtr=/Game/Athena/Sounds/Weapons/Fire_Axe_Swing_Distant_06}}
 {id=WavePlayer_46, properties {soundWaveAssetPtr=&Fire_Axe_Swing_Distant_03}}
 {id=WavePlayer_47, properties {soundWaveAssetPtr=&Fire_Axe_Swing_Distant_04}}
 {id=WavePlayer_48, properties {soundWaveAssetPtr=&Fire_Axe_Swing_Distant_06}}
 {id=WavePlayer_49, properties {soundWaveAssetPtr=&Fire_Axe_Swing_Distant_06}}
 {id=WavePlayer_50, properties {soundWaveAssetPtr=&/Moonlight_Assassin_Axe_Swing_01}}
 {id=WavePlayer_51, properties {soundWaveAssetPtr=&/Moonlight_Assassin_Axe_Swing_02}}
 {id=WavePlayer_52, properties {soundWaveAssetPtr=&/Moonlight_Assassin_Axe_Swing_03}}
 {id=WavePlayer_53, properties {soundWaveAssetPtr=&/Moonlight_Assassin_Axe_Swing_04}}
 {id=WavePlayer_54, properties {soundWaveAssetPtr=&/Moonlight_Assassin_Axe_Swing_05}}
 {id=WavePlayer_55, properties {soundWaveAssetPtr=&/Moonlight_Assassin_Axe_Swing_06}}
 {id=WavePlayer_56, properties {soundWaveAssetPtr=&Moonlight_Assassin_Axe_Swing_01}}
 {id=WavePlayer_57, properties {soundWaveAssetPtr=&Moonlight_Assassin_Axe_Swing_02}}
 {id=WavePlayer_58, properties {soundWaveAssetPtr=&Moonlight_Assassin_Axe_Swing_03}}
 {id=WavePlayer_59, properties {soundWaveAssetPtr=&Moonlight_Assassin_Axe_Swing_01}}
]
properties {
 audioPropertiesSheet=/BRCosmetics/Sounds/Settings/PropertySheets/Pickaxe/APS_Cosmetic_Pickaxe_Swing
 bOverrideConcurrency=true
 bPrimeOnLoad=true
 maxDistance=3000
 soundClassObject=&Fort_WeaponMelee_PSM
 volumeMultiplier=0.3
 concurrencyOverrides {bLimitToOwner=true, maxCount=2, resolutionRule=StopOldest}
 concurrencySet [&Player_Melee_Weapon_Conc]
}
